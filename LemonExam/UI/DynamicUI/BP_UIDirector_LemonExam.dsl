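; Dynamic UI director for LemonExam that shows or hides its UI scene as the LemonExam equipped gameplay tag is added or removed.
; On BeginPlay, outside compact UI, it starts tag waits on the owning player controller and again on each new view target.
; The add and remove tasks feed SetLemonExamUISceneActive, where a delta gate makes AddScene or RemoveScene run only when the state changes.
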(event EventBeginPlay
  (if (not (DisplayHelper|IsUsingCompactUI))
    (bind _asfort_player_controller (Utilities|Casting|CastToFortPlayerController (DynamicUIDirector|GetOwningLocalPlayerController)))
    (EventDispatchers|BindEventtoOnViewTargetChanged _asfort_player_controller (delegate Custom|OnViewTargetChanged))
    (CallFunction|WaitforGameplayTag _asfort_player_controller)))

(event Custom|OnViewTargetChanged
  (params (PlayerController :type Fort%20Player%20Controller%20Object%20Reference) (OldViewTarget :type Actor%20Object%20Reference) (NewViewTarget :type Actor%20Object%20Reference))
  (CallFunction|WaitforGameplayTag NewViewTarget))

(event Custom|WaitForGameplayTag
  (params (PlayerActor :type Actor%20Object%20Reference))
  (bind _returnvalue (DynamicUIDirector|GetOwningLocalPlayerController))
  ;[1 1 1] Wait For Appropriate Tag
  (Utilities|IsValid PlayerActor
    (:Is%20Valid
      (Utilities|IsValid _returnvalue
        (:Is%20Valid
          (Utilities|Casting|CastToFortPlayerPawnAthena (Pawn|GetControlledPawn _returnvalue))
          (Ability|Tasks|WaitGameplayTagAddToActor PlayerActor (Variables|Default|GetLemonExamEquippedTag)
            (:then
              (Ability|Tasks|WaitGameplayTagRemoveFromActor PlayerActor (Variables|Default|GetLemonExamEquippedTag)
                (:Removed
                  (Default|SetLemonExamUISceneActive false))))
            (:Added
              (Default|SetLemonExamUISceneActive true))))))))

(fn SetLemonExamUISceneActive
  (params (IsActive :type Boolean))
  (bind _uiscene (Variables|Default|GetUIScene))
  (bind _updated_variable (Utilities|FlowControl|DeltaGateWildcard (Variables|Default|GetIsLemonExamSceneActive) IsActive)
    (:Success
      (if _updated_variable
        (DynamicUIDirector|AddScene _uiscene)
        (else
          (DynamicUIDirector|RemoveScene _uiscene))))))

(fn ConstructionScript)
